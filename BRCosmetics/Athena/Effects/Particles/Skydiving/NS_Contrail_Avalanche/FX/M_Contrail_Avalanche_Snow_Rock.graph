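# Unlit material for Niagara mesh particles, evidently the snowy rock debris in the Avalanche skydiving contrail.
# It fakes sunlight by dotting the sun vector with a tiled, untransformed normal map for detail shading and with vertex normals for broad shadowing.
# Particle color and a time-of-day tint color the result, which feeds an emissive-adjustment function.
# Metalic, Specular and Roughness parameters are left unconnected.
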Material /BRCosmetics/Athena/Effects/Particles/Skydiving/NS_Contrail_Avalanche/FX/M_Contrail_Avalanche_Snow_Rock
graph {
 properties {bUsedWithNiagaraMeshParticles=true, pixelDepthOffsetMode=PDOM_Legacy, shadingModel=MSM_Unlit, usedShadingModels=Unlit}
 nodes [
  n1 = ScalarParameter() {defaultValue=0.2, parameterName=Shadow Multiplier}
  n2 = CollectionParameter() {
   collection=&FortniteMaterialParameters
   parameterName=SunlightDirectionalVector
  }
  n3 = ComponentMask(None=@n2) {b=true, g=true, r=true}
  n4 = TextureSampleParameter2D(UVs=TextureCoordinate() {uTiling=2}) {parameterName=Base Normal, samplerType=SAMPLERTYPE_Normal, texture=&T_Weap_Distort_N_02}
  n5 = ConstantBiasScale(None=DotProduct(A=@n3, B=@n4.RGB))
  n6 = Multiply(
   A=Multiply(A=ParticleColor().RGB, B=LinearInterpolate(Alpha=@n5) {constA=0.5})
   B=MaterialFunctionCall(
    Color_Morning=VectorParameter() {parameterName=Color Morning Multiplier, defaultValue=<rgba 0.9896 0.268 0.4929 0>}.RGB
    Color_Day=VectorParameter() {parameterName=Color Day Multiplier, defaultValue=<rgba 1 1 1 0>}.RGB
    Color_Evening=VectorParameter() {parameterName=Color Evening Multiplier, defaultValue=<rgba 0.9911 0.6442 0.7503 1>}.RGB
    Color_Night=VectorParameter() {parameterName=Color Night Multiplier, defaultValue=<rgba 0.3498 0.4864 0.6458 0>}.RGB
   ) {materialFunction=&Color_Time_of_Day}.Result
  )
  n7 = DotProduct(A=VertexNormalWS(), B=@n3)
  MaterialFunctionCall(Emissive=LinearInterpolate(A=Multiply(A=@n6, B=@n1), B=@n6, Alpha=ConstantBiasScale(None=@n7))) {materialFunction=&MF_FX_EmissiveAdjustments}
  ScalarParameter() {parameterName=Metalic}
  ScalarParameter() {defaultValue=0.5, parameterName=Specular}
  ScalarParameter() {defaultValue=0.5, parameterName=Roughness}
 ]
}
