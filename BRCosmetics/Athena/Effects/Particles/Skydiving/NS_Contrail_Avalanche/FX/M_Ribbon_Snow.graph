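# Translucent, unlit, two-sided snow ribbon material for the Avalanche skydiving contrail, usable on Niagara ribbons and mesh particles.
# Emissive multiplies particle color by a time-of-day tint through shared color and emissive functions.
# Opacity combines a panning dissolve texture, eroded by UV edge gradients, with a panning normal map dotted against Z for smoky breakup.
# Edge masks, inverted Fresnel, particle alpha and camera depth fade soften the result.
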Material /BRCosmetics/Athena/Effects/Particles/Skydiving/NS_Contrail_Avalanche/FX/M_Ribbon_Snow
graph {
 properties {
  bUsedWithNiagaraMeshParticles=true
  bUsedWithNiagaraRibbons=true
  blendMode=BLEND_Translucent
  pixelDepthOffsetMode=PDOM_Legacy
  shadingModel=MSM_Unlit
  twoSided=true
  usedShadingModels=Unlit
 }
 nodes [
  Constant() {r=1}
  n1 = ParticleColor()
  n2 = VectorParameter() {parameterName=Subsurface Color morning, defaultValue=<rgba 0.9911 0.2664 0.491 0>}
  n3 = VectorParameter() {parameterName=Subsurface Color day, defaultValue=<rgba 0.642 0.8085 0.9131 0>}
  n4 = VectorParameter() {parameterName=Subsurface Color evening, defaultValue=<rgba 0.5972 0.2918 0.3372 0>}
  n5 = VectorParameter() {parameterName=Subsurface Color night, defaultValue=<rgba 0.2542 0.4179 0.521 0>}
  MaterialFunctionCall(Emissive=Multiply(
    A=@n1.RGB
    B=MaterialFunctionCall(Color_Morning=@n2.RGB, Color_Day=@n3.RGB, Color_Evening=@n4.RGB, Color_Night=@n5.RGB) {materialFunction=&Color_Time_of_Day}.Result
   )) {materialFunction=&MF_FX_EmissiveAdjustments}
  n6 = ComponentMask(None=TextureCoordinate()) {g=true}
  n7 = Constant()
  n8 = Constant() {r=0.692}
  n9 = MaterialFunctionCall(Float2=TextureCoordinate()) {materialFunction=&BreakOutFloat2Components}
  n10 = Constant() {r=0.5}
  n11 = Saturate(None=SmoothStep(Min=@n7, Max=Add(A=@n7, B=@n8), Value=Distance(A=@n9.G, B=@n10)))
  n12 = ScalarParameter() {defaultValue=0.084, parameterName=Erode Edge Level}
  n13 = ScalarParameter() {defaultValue=1, parameterName=Erode Edge Softness}
  n14 = TextureCoordinate()
  n15 = Saturate(None=SmoothStep(
    Min=@n12
    Max=Add(A=@n12, B=@n13) {constB=0}
    Value=MaterialFunctionCall(Float2=@n14) {materialFunction=&BreakOutFloat2Components}.R
   ))
  n16 = Constant() {r=1}
  n17 = ScalarParameter() {defaultValue=1, parameterName=Erode Texture Softness}
  n18 = Multiply(A=MaterialFunctionCall(Input=Max(A=@n11, B=@n15), Input Low=Constant(), Input High=@n16, Target Low=Multiply(A=@n17) {constB=-1}, Target High=@n16) {materialFunction=&RemapValueRange}.Result)
  n19 = Multiply(A=TextureCoordinate() {uTiling=2}, B=Constant2Vector() {g=1, r=1})
  n20 = SmoothStep(
   Min=@n18
   Max=Add(A=@n18, B=@n17)
   Value=TextureSample(UVs=Panner(Coordinate=@n19) {speedX=-0.5}) {texture=/Game/CharClasses/Outlander/Gadgets/AntiMaterialCharge/Materials/T_AntiMaterialCharge_Dissolve_01}.R
  )
  n21 = ScalarParameter() {parameterName=Soft Edge}
  n22 = Add(A=@n21, B=ScalarParameter() {defaultValue=0.1, parameterName=Edge Fade})
  n23 = MaterialFunctionCall(Float2=TextureCoordinate()) {materialFunction=&BreakOutFloat2Components}
  n24 = Multiply(A=Saturate(None=Sine(None=@n6) {period=2}), B=Multiply(A=Saturate(None=@n20), B=SmoothStep(Min=@n21, Max=@n22, Value=@n23.R)))
  n25 = Constant3Vector() {constant=<rgba 0 0 1 0>}
  n26 = Multiply(A=TextureCoordinate(), B=Constant2Vector() {g=0.7, r=0.8})
  n27 = Constant2Vector() {g=0.15}
  n28 = TextureSample(UVs=Panner(Coordinate=Add(A=@n26, B=@n27)) {speedX=-1}) {samplerType=SAMPLERTYPE_Normal, texture=/Game/Effects/Fort_Effects/Gameplay/SmokeDustExplosions/Textures/T_RibbonSmokeMasked_01}
  n29 = Saturate(None=Power(Base=DotProduct(A=@n25, B=@n28.RGB), Exp=ScalarParameter() {defaultValue=3, parameterName=Power}) {constExponent=3})
  n30 = Multiply(A=@n1.A, B=Saturate(None=Multiply(A=@n24, B=@n29)))
  n31 = OneMinus(None=Saturate(None=Fresnel() {baseReflectFraction=0, exponent=0.1}))
  Multiply(
   A=Multiply(A=Multiply(A=@n30, B=@n31), B=ScalarParameter() {defaultValue=1, parameterName=Alpha})
   B=MaterialFunctionCall(
    Fade Length=ScalarParameter() {defaultValue=256, parameterName=Fade Length}
    Fade Offset=ScalarParameter() {defaultValue=150, parameterName=Fade Offset}
   ) {materialFunction=&CameraDepthFade}.Result
  )
 ]
}
