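# Avalanche skydiving contrail cosmetic that swirls snow chunks and trails around the player.
# Snow_Chunks meshes and Snow_Guide particles move under drag, inherited movement, sphere blocking and vortex force scaled by system rotational speed.
# Fragment, puff, smoke and ribbon emitters spawn from those leaders via particle attribute readers.
# Glide lean alpha drives offsets and rates, time of day and front-end modifiers adjust the look, and old spring-based emitters are disabled.
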NiagaraSystem /BRCosmetics/Athena/Effects/Particles/Skydiving/NS_Contrail_Avalanche/FX/NS_Contrail_Avalanche_01
dependencies {
 usedDataInterfaces [&NiagaraDataInterfaceCurve, &NiagaraDataInterfaceParticleRead]
 usedDynamicInputs []
 usedModules [
  &self:SystemSpawnScript.SystemScriptSource.SystemScriptGraph.NiagaraNodeAssignment_6.SetVariables_#1
  &self:SystemSpawnScript.SystemScriptSource.SystemScriptGraph.NiagaraNodeAssignment_7.SetVariables_#2
  &self:SystemSpawnScript.SystemScriptSource.SystemScriptGraph.NiagaraNodeAssignment_3.SetVariables_#3
  &self:CalculateRotationalSpeed
  &self:SystemSpawnScript.SystemScriptSource.SystemScriptGraph.NiagaraNodeAssignment_4.SetVariables_#4
  &SystemState
 ]
}
emitters {
 Fragments_Chunks {
  emitterUpdateScript [
   &EmitterState()
   &SpawnParticlesFromOtherEmitter(
    Attribute Reader=DataInterface() {propertyValues {EmitterBinding {bindingMode=Other, emitterName=Snow_Chunks}}}
    Spawn Rate Per Particle Cap=10
    Use Maximum Age=false
    Use Minimum Age=false
   )
  ]
  particleSpawnScript [
   &InitializeParticle(
    Color=<rgba 0.4492 0.6108 0.8177 1>
    Lifetime Max=0.8
    Lifetime Min=0.25
    Lifetime Mode=::NewEnumerator1
    Mesh Scale=&Multiply_Vector(
     A=&Multiply_VectorByFloat(
      Vector=&RandomRangeVector(Minimum=<0.2 0.1 0.1>, Maximum=<0.5 0.3 0.3>)
      Float=&RandomRangeFloat(Minimum=0.03, Maximum=0.05)
     )
     B=&VectorFromFloat(Value=&FN_FrontEndFloatModifier(Gameplay=1, Front End=0.8))
    )
    Mesh Scale Mode=::NewEnumerator4
    Sprite Rotation Mode=::NewEnumerator2
    Sprite Size Mode=::NewEnumerator1
    Uniform Sprite Size Max=20
    Uniform Sprite Size Min=20
    UsePositionOffset=false
    Write Material Random=true
   )
   &SampleParticlesFromOtherEmitter(
    Applied Velocity Scale=-0.1
    Minimum Sampled Particle Age=0.3
    Module.MaxSampledAgeCutoff=false
    Module.MinSampledAgeCutoff=true
    Reject Particle By Sampled Age=true
    Velocity Sampling=::NewEnumerator0
   )
   &AddVelocityInCone(Cone Angle=20, Cone Axis=<0 0 1>, Module.Use Velocity Falloff On Cone Axis=false, Velocity Strength=300) {module=AddVelocityInCone001}
   &AddVelocity() {module=AddVelocity002}
   &CameraOffset(Camera Offset Amount=10)
   &SphereLocation(
    MaskSpawnGroup=false
    Offset=<0 0 15>
    Override Seed=false
    Sphere Radius=10
    Surface Only=false
    UseNonUniformScale=false
    UseOffset=true
    UseSphereOrientation=false
   )
   &InitialMeshOrientation(
    Mesh Orientation Mode=::NewEnumerator2
    Rotation=&RandomRangeVector(Minimum=<0 0 0>, Maximum=<1 0 0>)
    Use Rotation Vector=true
   ) {module=InitialMeshOrientation001}
  ]
  particleUpdateScript [
   &InheritSourceMovement(ApplyForce=false, ApplyPosition=true, ApplyPositionFalloff=false)
   &SolveForcesAndVelocity(Clamp Velocity=false, Limit Acceleration=false)
   &ParticleState()
   &ScaleColor(
    Scale Alpha=&Multiply_Float(
     A=&FloatFromCurve(
      FloatCurve=DataInterface() {propertyValues {Curve {keys [
          {interpMode=RCIM_Cubic, tangentMode=RCTM_User, time=0, value=1.5}
          {interpMode=RCIM_Cubic, leaveTangent=-2.6773, tangentMode=RCTM_Break, time=0.3, value=1.5}
          {tangentMode=RCTM_User, time=1, value=0}
         ]}}}
      Scale Curve=&RandomRangeFloat(Minimum=0.5, Maximum=1)
     )
     B=&FNTimeOfDayScalar(Night Scale=1, Morning Scale=0.2, Evening Scale=0.2)
    )
    ScaleA=true
    ScaleRGB=true
   )
   &ScaleMeshSize(Scale Factor=&VectorFromFloat(Value=&FloatFromCurve(FloatCurve=DataInterface() {propertyValues {Curve {keys [
          {arriveTangent=9.0725, interpMode=RCIM_Cubic, leaveTangent=9.0725, tangentMode=RCTM_User, time=0, value=0}
          {arriveTangent=-2.5249, interpMode=RCIM_Cubic, leaveTangent=-2.5249, tangentMode=RCTM_User, time=0.1408, value=0.9847}
          {arriveTangent=-0.0405, interpMode=RCIM_Cubic, leaveTangent=-0.0405, tangentMode=RCTM_User, time=1, value=0}
         ]}}})))
  ]
  rendererClasses [&NiagaraMeshRendererProperties]
  renderers [{rendererClass=&NiagaraMeshRendererProperties, rendererIndex=0}]
 }
 Fragments_Guide {
  emitterUpdateScript [
   &EmitterState()
   &SpawnParticlesFromOtherEmitter(
    Attribute Reader=DataInterface() {propertyValues {EmitterBinding {bindingMode=Other, emitterName=Snow_Guide}}}
    Spawn Rate Per Particle Cap=10
    Use Maximum Age=false
    Use Minimum Age=false
   )
  ]
  particleSpawnScript [
   &InitializeParticle(
    Color=<rgba 0.4492 0.6108 0.8177 1>
    Lifetime Max=0.5
    Lifetime Min=0.25
    Lifetime Mode=::NewEnumerator1
    Mesh Scale=&Multiply_Vector(
     A=&Multiply_VectorByFloat(
      Vector=&RandomRangeVector(Minimum=<0.2 0.1 0.1>, Maximum=<0.5 0.3 0.3>)
      Float=&RandomRangeFloat(Minimum=0.03, Maximum=0.05)
     )
     B=&VectorFromFloat(Value=&FN_FrontEndFloatModifier(Gameplay=1, Front End=0.8))
    )
    Mesh Scale Mode=::NewEnumerator4
    Sprite Rotation Mode=::NewEnumerator2
    Sprite Size Mode=::NewEnumerator1
    Uniform Sprite Size Max=20
    Uniform Sprite Size Min=20
    UsePositionOffset=false
    Write Material Random=true
   )
   &SampleParticlesFromOtherEmitter(Applied Velocity Scale=-0.1, Velocity Sampling=::NewEnumerator0)
   &AddVelocityInCone(Cone Angle=20, Cone Axis=<0 0 1>, Module.Use Velocity Falloff On Cone Axis=false, Velocity Strength=300) {module=AddVelocityInCone001}
   &AddVelocity(Velocity=<0 0 100>) {module=AddVelocity002}
   &CameraOffset(Camera Offset Amount=10)
   &SphereLocation(
    MaskSpawnGroup=false
    Offset=<0 0 15>
    Override Seed=false
    Sphere Radius=25
    Surface Only=false
    UseNonUniformScale=false
    UseOffset=true
    UseSphereOrientation=false
   )
   /SetVariables(Particles.ColorLow:LinearColor=<rgba 0.5209 0.6876 0.901 1>) {isSetParameters=true}
   &InitialMeshOrientation(
    Mesh Orientation Mode=::NewEnumerator2
    Rotation=&RandomRangeVector(Minimum=<0 0 0>, Maximum=<1 0 0>)
    Use Rotation Vector=true
   ) {module=InitialMeshOrientation001}
  ]
  particleUpdateScript [
   &InheritSourceMovement(ApplyForce=false, ApplyPosition=true, ApplyPositionFalloff=false)
   &SolveForcesAndVelocity(Clamp Velocity=false, Limit Acceleration=false)
   &ParticleState()
   &ScaleColor(
    Scale Alpha=&Multiply_Float(
     A=&FloatFromCurve(
      FloatCurve=DataInterface() {propertyValues {Curve {keys [
          {interpMode=RCIM_Cubic, tangentMode=RCTM_User, time=0, value=1}
          {interpMode=RCIM_Cubic, leaveTangent=-2.6773, tangentMode=RCTM_Break, time=0.3, value=1}
          {tangentMode=RCTM_User, time=1, value=0}
         ]}}}
      Scale Curve=&RandomRangeFloat(Minimum=0.5, Maximum=1)
     )
     B=&FNTimeOfDayScalar(Night Scale=1, Morning Scale=0.2, Evening Scale=0.2)
    )
    ScaleA=true
    ScaleRGB=true
   )
   &ScaleMeshSize(Scale Factor=&VectorFromFloat(Value=&FloatFromCurve(FloatCurve=DataInterface() {propertyValues {Curve {keys [<at 0 0>,<at 0.1 1>,<at 1 0>]}}})))
  ]
  rendererClasses [&NiagaraMeshRendererProperties]
  renderers [{rendererClass=&NiagaraMeshRendererProperties, rendererIndex=0}]
 }
 Puff_Snow {
  emitterUpdateScript [
   &EmitterState()
   &SpawnParticlesFromOtherEmitter(
    Attribute Reader=DataInterface() {propertyValues {EmitterBinding {bindingMode=Other, emitterName=Snow_Chunks}}}
    Spawn Rate=&Lerp_Float(A=15, B=20, Alpha=$User.ParaGlideLeanAlpha)
    Spawn Rate Per Particle Cap=10000000
    Use Maximum Age=false
    Use Minimum Age=false
   )
  ]
  particleSpawnScript [
   &InitializeParticle(
    Color=<rgba 0.5537 0.6694 0.8177 0.8>
    Lifetime Max=0.4
    Lifetime Min=0.3
    Lifetime Mode=::NewEnumerator1
    Sprite Rotation Mode=::NewEnumerator2
    Sprite Size Mode=::NewEnumerator1
    Uniform Sprite Size Max=55
    Uniform Sprite Size Min=45
    UsePositionOffset=false
    Write Material Random=true
   )
   &SampleParticlesFromOtherEmitter(Applied Velocity Scale=-0.1, Velocity Sampling=::NewEnumerator0)
   /SetVariables(Particles.Color=&MultiplyLinearColor(
     Linear Color=$Particles.Initial.Color
     Scale Factor=&MakeVector4_FromLinearColor(LinearColor=&FN_Select_Color_from_Time_Of_Day(Morning=<rgba 0.7605 0.491 0.6172 1>, Day=<rgba 1 1 1 1>, Evening=<rgba 0.7605 0.491 0.6172 1>, Night=<rgba 0.6516 0.7281 0.8177 1>))
    )) {isSetParameters=true}
   &AddVelocity(Velocity=<0 0 200>)
   &SphereLocation(
    MaskSpawnGroup=false
    Override Seed=false
    Sphere Radius=0.3
    Surface Only=false
    UseNonUniformScale=false
    UseOffset=false
    UseSphereOrientation=false
   )
  ]
  particleUpdateScript [
   &InheritSourceMovement(ApplyForce=false, ApplyPosition=true, ApplyPositionFalloff=false)
   &SolveForcesAndVelocity(Clamp Velocity=false, Limit Acceleration=false)
   &ParticleState()
   &ScaleColor(
    Scale Alpha=&Multiply_Float(
     A=&FloatFromCurve(
      FloatCurve=DataInterface() {propertyValues {Curve {keys [<at 0.1 0 RCIM_Cubic>, <at 0.35 1 RCIM_Cubic>, {tangentMode=RCTM_User, time=1, value=0}]}}}
      Scale Curve=&RandomRangeFloat(Minimum=0.8, Maximum=1)
     )
     B=&FNTimeOfDayScalar(Night Scale=0.4, Morning Scale=0.3, Day Scale=0.8, Evening Scale=0.3)
    )
    ScaleA=true
    ScaleRGB=true
   )
   &ScaleSpriteSize(Uniform Curve Sprite Scale=DataInterface() {propertyValues {Curve {keys [{arriveTangent=1.2187, time=0, value=0.35}, {leaveTangent=3.3905, time=1, value=1}]}}})
  ]
  rendererClasses [&NiagaraSpriteRendererProperties]
  renderers [{rendererClass=&NiagaraSpriteRendererProperties}]
 }
 Puff_Snow_Guide {
  emitterUpdateScript [
   &EmitterState()
   &SpawnParticlesFromOtherEmitter(
    Attribute Reader=DataInterface() {propertyValues {EmitterBinding {bindingMode=Other, emitterName=Snow_Guide}}}
    Spawn Rate=&Lerp_Float(A=8, B=8, Alpha=$User.ParaGlideLeanAlpha)
    Spawn Rate Per Particle Cap=10000000
    Use Maximum Age=false
    Use Minimum Age=false
   )
  ]
  particleSpawnScript [
   &InitializeParticle(
    Color=<rgba 0.5537 0.6694 0.8177 0.8>
    Lifetime Max=0.45
    Lifetime Min=0.35
    Lifetime Mode=::NewEnumerator1
    Sprite Rotation Mode=::NewEnumerator2
    Sprite Size Mode=::NewEnumerator3
    Uniform Sprite Size=90
    UsePositionOffset=false
    Write Material Random=true
   )
   &SampleParticlesFromOtherEmitter(Applied Velocity Scale=-0.1, Velocity Sampling=::NewEnumerator0)
   &AddVelocity(Velocity=<0 0 300>)
   /SetVariables(Particles.Color=&MultiplyLinearColor(
     Linear Color=$Particles.Initial.Color
     Scale Factor=&MakeVector4_FromLinearColor(LinearColor=&FN_Select_Color_from_Time_Of_Day(Morning=<rgba 0.7605 0.491 0.6172 1>, Day=<rgba 1 1 1 1>, Evening=<rgba 0.7605 0.491 0.6172 1>, Night=<rgba 0.6516 0.7281 0.8177 1>))
    )) {isSetParameters=true}
   &SphereLocation(
    MaskSpawnGroup=false
    Override Seed=false
    Sphere Radius=0.3
    Surface Only=false
    UseNonUniformScale=false
    UseOffset=false
    UseSphereOrientation=false
   )
  ]
  particleUpdateScript [
   &ParticleState()
   &InheritSourceMovement(ApplyForce=false, ApplyPosition=true, ApplyPositionFalloff=false)
   &SolveForcesAndVelocity(Clamp Velocity=false, Limit Acceleration=false)
   &ScaleColor(
    Scale Alpha=&Multiply_Float(
     A=&FloatFromCurve(
      FloatCurve=DataInterface() {propertyValues {Curve {keys [<at 0.25 0 RCIM_Cubic>, <at 0.45 1 RCIM_Cubic>, {tangentMode=RCTM_User, time=1, value=0}]}}}
      Scale Curve=&RandomRangeFloat(Minimum=0.5, Maximum=1)
     )
     B=&FNTimeOfDayScalar(Night Scale=0.4, Morning Scale=0.3, Day Scale=0.8, Evening Scale=0.3)
    )
    ScaleA=true
    ScaleRGB=true
   )
   &ScaleSpriteSize(Uniform Curve Sprite Scale=DataInterface() {propertyValues {Curve {keys [<at 0 0.1>, {leaveTangent=3.3905, time=1, value=1}]}}})
  ]
  rendererClasses [&NiagaraSpriteRendererProperties]
  renderers [{rendererClass=&NiagaraSpriteRendererProperties}]
 }
 Smoke {
  emitterUpdateScript [
   &EmitterState()
   &SpawnParticlesFromOtherEmitter(
    Attribute Reader=DataInterface() {propertyValues {EmitterBinding {bindingMode=Other, emitterName=Snow_Guide}}}
    Spawn Rate=&Lerp_Float(A=50, B=80, Alpha=$User.ParaGlideLeanAlpha)
    Spawn Rate Per Particle Cap=10000000
    Use Maximum Age=false
    Use Minimum Age=false
   )
  ]
  particleSpawnScript [
   &InitializeParticle(
    Color=<rgba 1 1 1 0.45>
    Lifetime Max=0.45
    Lifetime Min=0.35
    Lifetime Mode=::NewEnumerator1
    Sprite Rotation Mode=::NewEnumerator2
    Sprite Size Mode=::NewEnumerator3
    Uniform Sprite Size=80
    UsePositionOffset=false
    Write Material Random=true
   )
   &SampleParticlesFromOtherEmitter(Applied Velocity Scale=-0.1, Velocity Sampling=::NewEnumerator0)
   &AddVelocity(Velocity=<0 0 300>)
   &SphereLocation(
    MaskSpawnGroup=false
    Override Seed=false
    Sphere Radius=0.3
    Surface Only=false
    UseNonUniformScale=false
    UseOffset=false
    UseSphereOrientation=false
   )
  ]
  particleUpdateScript [
   &ParticleState()
   &InheritSourceMovement(ApplyForce=false, ApplyPosition=true, ApplyPositionFalloff=false)
   &SolveForcesAndVelocity(Clamp Velocity=false, Limit Acceleration=false)
   &ScaleColor(
    Scale Alpha=&Multiply_Float(
     A=&FloatFromCurve(
      FloatCurve=DataInterface() {propertyValues {Curve {keys [<at 0 0 RCIM_Cubic>,<at 0 0 RCIM_Cubic>,<at 0.25 1>,<at 1 0 RCIM_Cubic>]}}}
      Scale Curve=&RandomRangeFloat(Minimum=0.7, Maximum=1)
     )
     B=&FNTimeOfDayScalar(Night Scale=1.5, Morning Scale=1, Evening Scale=1)
    )
    Scale RGB=&Lerp_Vector(
     A=&MakeVectorFromLinearColorRGB(Color=<rgba 0.3863 0.6754 0.9271 1>)
     B=&MakeVectorFromLinearColorRGB(Color=<rgba 0.6867 0.7454 0.8148 1>)
     Alpha=&FloatFromCurve(FloatCurve=DataInterface() {propertyValues {Curve {keys [<at 0 0 RCIM_Cubic>,<at 0.7 1 RCIM_Cubic>]}}})
    )
    ScaleA=true
    ScaleRGB=true
   )
   &ScaleSpriteSize(Uniform Curve Sprite Scale=DataInterface() {propertyValues {Curve {keys [{arriveTangent=1.2187, time=0, value=0.2}, {leaveTangent=3.3905, time=1, value=1.2}]}}})
  ]
  rendererClasses [&NiagaraSpriteRendererProperties]
  renderers [
   {rendererClass=&NiagaraSpriteRendererProperties}
   {rendererClass=&NiagaraSpriteRendererProperties, rendererIndex=1}
  ]
 }
 Smoke001 {
  emitterUpdateScript [
   &EmitterState()
   &SpawnParticlesFromOtherEmitter(
    Attribute Reader=DataInterface() {propertyValues {EmitterBinding {bindingMode=Other, emitterName=Snow_Chunks}}}
    Spawn Rate=&Lerp_Float(A=70, B=110, Alpha=$User.ParaGlideLeanAlpha)
    Spawn Rate Per Particle Cap=10000000
    Use Maximum Age=false
    Use Minimum Age=false
   )
  ]
  particleSpawnScript [
   &InitializeParticle(
    Color=<rgba 1 1 1 0.45>
    Lifetime Max=0.45
    Lifetime Min=0.35
    Lifetime Mode=::NewEnumerator1
    Sprite Rotation Mode=::NewEnumerator2
    Sprite Size Mode=::NewEnumerator1
    Uniform Sprite Size Max=46
    Uniform Sprite Size Min=44
    UsePositionOffset=false
    Write Material Random=true
   )
   &SampleParticlesFromOtherEmitter(Applied Velocity Scale=-0.1, Velocity Sampling=::NewEnumerator0)
   &AddVelocity(Velocity=<0 0 200>)
   &SphereLocation(
    MaskSpawnGroup=false
    Override Seed=false
    Sphere Radius=0.3
    Surface Only=false
    UseNonUniformScale=false
    UseOffset=false
    UseSphereOrientation=false
   )
  ]
  particleUpdateScript [
   &ParticleState()
   &InheritSourceMovement(ApplyForce=false, ApplyPosition=true, ApplyPositionFalloff=false)
   &SolveForcesAndVelocity(Clamp Velocity=false, Limit Acceleration=false)
   &ScaleColor(
    Scale Alpha=&Multiply_Float(
     A=&FloatFromCurve(
      FloatCurve=DataInterface() {propertyValues {Curve {keys [{interpMode=RCIM_Cubic, leaveTangent=-2.6773, tangentMode=RCTM_Break, time=0.15, value=1}, {tangentMode=RCTM_User, time=1, value=0}]}}}
      Scale Curve=&RandomRangeFloat(Minimum=0.7, Maximum=1)
     )
     B=&FNTimeOfDayScalar(Night Scale=1.3, Morning Scale=1, Evening Scale=1)
    )
    Scale RGB=&Lerp_Vector(
     A=&MakeVectorFromLinearColorRGB(Color=<rgba 0.3863 0.6754 0.9271 1>)
     B=&MakeVectorFromLinearColorRGB(Color=<rgba 0.6867 0.7454 0.8148 1>)
     Alpha=&FloatFromCurve(FloatCurve=DataInterface() {propertyValues {Curve {keys [<at 0 0 RCIM_Cubic>,<at 0.7 1 RCIM_Cubic>]}}})
    )
    ScaleA=true
    ScaleRGB=true
   )
   &ScaleSpriteSize(Uniform Curve Sprite Scale=DataInterface() {propertyValues {Curve {keys [{arriveTangent=1.2187, time=0, value=0.3}, {leaveTangent=3.3905, time=1, value=1.2}]}}})
  ]
  rendererClasses [&NiagaraSpriteRendererProperties]
  renderers [
   {rendererClass=&NiagaraSpriteRendererProperties}
   {rendererClass=&NiagaraSpriteRendererProperties, rendererIndex=1}
  ]
 }
 Snow_Chunks {
  emitterUpdateScript [
   &EmitterState(
    Inactive Response=::NewEnumerator2
    Life Cycle Mode=::NewEnumerator1
    Loop Behavior=::NewEnumerator1
    Loop Delay=0.1
    Loop Duration Mode=::NewEnumerator1
    Scalability Mode=::NewEnumerator1
    UseLoopDelay=false
   )
   &SpawnRate(Spawn Probability=0.4, SpawnRate=1.8, Use Spawn Probability=false)
  ]
  particleSpawnScript [
   &InitializeParticle(
    Color=&MultiplyLinearColor(
     Linear Color=<rgba 1 1 1 1>
     Scale Factor=&MakeVector4_FromFloat(Float=&FN_FrontEndFloatModifier(Gameplay=1, Front End=2.222))
    )
    Mass Mode=::NewEnumerator1
    Mesh Scale Mode=::NewEnumerator3
    Mesh Uniform Scale=&RandomRangeFloat(Minimum=0.4, Maximum=0.35)
    UsePositionOffset=false
    Write Material Random=true
   )
   /Niagara/Modules/Update/Mass/CalculateSizeAndRotationalInertiaByMass(Density by Material Type:ENiagaraMassByVolume=::NewEnumerator2, "Manually Entered Density (kg/m^3)"=10, Module.Manually Enter Density=false)
   /SetVariables(Particles.MeshIndex=&RandomRangeInt(Minimum=0, Maximum=3)) {isSetParameters=true, module=SetVariables_#5}
   &ShapeLocation(
    MaskSpawnGroup=false
    Non Uniform Scale=<1.1 -2 1>
    Offset=&Lerp_Vector(A=<0 -180 300>, B=<0 -180 450>, Alpha=$User.ParaGlideLeanAlpha)
    Offset Coordinate Space=::Local
    Offset Mode=::NewEnumerator0
    Override Seed=false
    Random Seed=220
    Shape Primitive=::NewEnumerator4
    U Distribution=0.5
    UseOwnerScale=true
   )
   &InitialMeshOrientation(Use Rotation Vector=false)
   &ApplyInitialForces(Apply Force to Position=true)
  ]
  particleUpdateScript [
   &ParticleState()
   &Drag(
    Drag=&Multiply_Float(
     A=20
     B=&FloatFromCurve(FloatCurve=DataInterface() {propertyValues {Curve {keys [{interpMode=RCIM_Cubic, tangentMode=RCTM_User, time=0, value=0}, {interpMode=RCIM_Cubic, tangentMode=RCTM_User, time=0.5, value=1}]}}})
    )
    Use Linear Drag=true
    Use Rotational Drag=false
   )
   &InheritSourceMovement(
    Applied Position Scale=&MakeVector(
     X=1
     Y=1
     Z=&Multiply_Float(
      A=&Lerp_Float(A=1, B=1.03, Alpha=$User.ParaGlideLeanAlpha)
      B=&Lerp_Float(
       A=&FloatFromCurve(FloatCurve=DataInterface() {propertyValues {Curve {keys [{arriveTangent=-3.9401, interpMode=RCIM_Cubic, leaveTangent=-3.9401, tangentMode=RCTM_User, time=0, value=1.6}, <at 0.2 1 RCIM_Cubic>]}}})
       B=&FloatFromCurve(FloatCurve=DataInterface() {propertyValues {Curve {keys [{arriveTangent=-3.0844, interpMode=RCIM_Cubic, leaveTangent=-3.0844, tangentMode=RCTM_User, time=0, value=1.3}, <at 0.2 1 RCIM_Cubic>]}}})
       Alpha=$User.ParaGlideLeanAlpha
      )
     )
    )
    ApplyForce=false
    ApplyPosition=true
    ApplyPositionFalloff=false
   ) {module=InheritSourceMovement002}
   &self:SphereBlockingVolume_02(Sphere Origin:Vector3f=/Niagara/DynamicInputs/Helpers/SimulationPosition(), Sphere Radius=70) {module=SphereBlockingVolume_02}
   &VortexForce(
    UseOriginOffset=false
    Vortex Force Amount=&Multiply_Float(A=&RandomRangeFloat(Minimum=2400, Maximum=2700), B=$System.RotationalSpeed)
   )
   /SetVariables(Particles.Velocity=&Add_Vector(
     A=$Particles.Velocity
     B=&MakeVector(
      X=0
      Y=0
      Z=&FloatFromCurve(
       FloatCurve=DataInterface() {propertyValues {Curve {keys [
           {arriveTangent=7.3676, time=0, value=0.1}
           {arriveTangent=2.8333, leaveTangent=2.8333, time=0.08, value=0.92}
           {arriveTangent=0.0699, interpMode=RCIM_Cubic, leaveTangent=0.0699, tangentMode=RCTM_User, time=0.3, value=0.95}
           <at 1 1 RCIM_Cubic>
          ]}}}
       Scale Curve=&Multiply_Float(
        A=&RandomRangeFloat(Minimum=-80, Maximum=-100)
        B=&FN_FrontEndFloatModifier(Gameplay=1, Front End=&Lerp_Float(A=1.5, B=2, Alpha=$User.ParaGlideLeanAlpha))
       )
      )
     )
    )) {isSetParameters=true, module=SetVariables_#6}
   /SetVariables(Particles.Velocity=&Add_Vector(
     A=$Particles.Velocity
     B=&MakeVector(
      X=0
      Y=0
      Z=&FloatFromCurve(
       FloatCurve=DataInterface() {propertyValues {Curve {keys [<at 0.7 0>,<at 1 1>]}}}
       Scale Curve=&RandomRangeFloat(Minimum=180, Maximum=250)
      )
     )
    )) {isSetParameters=true, module=SetVariables_#7}
   &AccelerationForce(Acceleration=&Multiply_VectorByFloat(
     Vector=&TransformVector(Vector=<0 1 0>, Destination Space=::World)
     Float=&FloatFromCurve(
      FloatCurve=DataInterface() {propertyValues {Curve {keys [
          {arriveTangent=-4.5927, interpMode=RCIM_Cubic, leaveTangent=-4.5927, tangentMode=RCTM_User, time=0.2, value=1}
          {interpMode=RCIM_Cubic, tangentMode=RCTM_User, time=0.8, value=-1}
         ]}}}
      Scale Curve=&Multiply_Float(A=&RandomRangeFloat(Minimum=2000, Maximum=2500), B=1)
     )
    )) {module=AccelerationForce002}
   &UpdateMeshOrientation(
    Rotation Rate=&Multiply_Float(
     A=&Divide_Float(A=5, B=$Particles.Mass)
     B=&RandomRangeFloat(Minimum=0.3, Maximum=0.5)
    )
    Rotation Vector=&RandomVector(Vector Scale=1)
   )
   &SolveForcesAndVelocity(Clamp Velocity=true, Limit Acceleration=true)
   &ScaleMeshSize(Scale Factor=&Multiply_VectorByFloat(
     Vector=$Engine.Owner.Scale
     Float=&FloatFromCurve(FloatCurve=DataInterface() {propertyValues {Curve {keys [
          <at 0 0 RCIM_Cubic>
          <at 0.2 0.5 RCIM_Cubic>
          {arriveTangent=-0.2898, interpMode=RCIM_Cubic, leaveTangent=-0.2898, tangentMode=RCTM_User, time=0.9, value=0.2}
          <at 1 0>
         ]}}})
    ))
  ]
  rendererClasses [&NiagaraMeshRendererProperties]
  renderers [{rendererClass=&NiagaraMeshRendererProperties, rendererIndex=0}]
 }
 Snow_Chunks_old_spring {
  bEnabled=false
  emitterUpdateScript [
   &EmitterState(
    Inactive Response=::NewEnumerator2
    Life Cycle Mode=::NewEnumerator1
    Loop Behavior=::NewEnumerator1
    Loop Duration Mode=::NewEnumerator1
    UseLoopDelay=false
   )
   &SpawnRate(SpawnRate=2.8, Use Spawn Probability=false)
  ]
  particleSpawnScript [
   &InitializeParticle(
    Color=&MultiplyLinearColor(
     Linear Color=<rgba 0.0075 0.0245 0.05 1>
     Scale Factor=&MakeVector4_FromFloat(Float=&FN_FrontEndFloatModifier(Gameplay=1, Front End=2.222))
    )
    Mass=&Lerp_Float(
     A=1
     B=2
     Alpha=&FloatFromCurve(
      FloatCurve=DataInterface() {propertyValues {Curve {keys [
          {interpMode=RCIM_Cubic, tangentMode=RCTM_User, time=0, value=0}
          {arriveTangent=3.2438, interpMode=RCIM_Cubic, leaveTangent=3.2438, tangentMode=RCTM_User, time=1, value=1}
         ]}}}
      CurveIndex=&RandomRangeFloat(Minimum=0, Maximum=1)
     )
    )
    Mass Mode=::NewEnumerator1
    Mesh Scale Mode=::NewEnumerator3
    Mesh Uniform Scale=&RandomRangeFloat(Minimum=0.18, Maximum=0.2)
    UsePositionOffset=false
    Write Material Random=true
   )
   &CalculateSizeAndRotationalInertiaByMass(Density by Material Type:ENiagaraMassByVolume=::NewEnumerator2, "Manually Entered Density (kg/m^3)"=10, Module.Manually Enter Density=false)
   &TorusLocation(
    Align Torus With Axis=true
    Coordinate Space:ENiagaraCoordinateSpace=::Simulation
    Handle Radius=40
    MaskSpawnGroup=false
    Module.Override Randomness=false
    Normalized Angle Around Torus Axis=0.2
    Offset=<0 0 300>
    Torus Axis=&RandomRangeVector(Minimum=<0.1 -0.1 1>, Maximum=<-0.1 0.1 1>)
    Torus Origin:NiagaraPosition=&SimulationPosition()
   )
   /SetVariables(
    Particles.InitialTorusOffset:Vector3f=&Add_Vector(A=$Output.TorusLocation.TorusVector, B=<0 0 300>)
    Particles.MeshIndex=&RandomRangeInt(Minimum=0, Maximum=1)
   ) {isSetParameters=true, module=SetVariables_#5}
   &InitialMeshOrientation(Use Rotation Vector=false)
   &ApplyInitialForces(Apply Force to Position=true)
  ]
  particleUpdateScript [
   &ParticleState()
   &InterpolateOverTime(
    Initial Vector Value:Vector3f=$Particles.Initial.InitialTorusOffset
    Rate Of Change=35
    Target Vector Value:Vector3f=&TransformVector(Vector=$Particles.Initial.InitialTorusOffset, Destination Space=::Simulation, Apply Scale=true)
    Variable Type=::NewEnumerator2
   )
   /SetVariables(Particles.InterpolatedTorusOffset:Vector3f=&Multiply_VectorByFloat(
     Vector=&NormalizeVector(
      Vector To Normalize=$StackContext.InterpolateOverTime.InterpolateOverTime_V3_.Moving Average
      Safe Normalize=true
      Safe Fallback Threshold=0
      Safe Fallback Vector=<0 0 0.001>
     )
     Float=&VectorLength(Vector=$Particles.Initial.InitialTorusOffset)
    )) {isSetParameters=true, module=SetVariables_#8}
   &Drag(
    Drag=&Multiply_Float(
     A=20
     B=&FloatFromCurve(FloatCurve=DataInterface() {propertyValues {Curve {keys [{interpMode=RCIM_Cubic, tangentMode=RCTM_User, time=0, value=-0.1}, {interpMode=RCIM_Cubic, tangentMode=RCTM_User, time=0.5, value=1}]}}})
    )
    Use Linear Drag=true
    Use Rotational Drag=false
   )
   /Niagara/Modules/Update/Forces/SpringForce(
    " Particle Equilibrium Position":NiagaraPosition=&Add_Vector(
     A=&SimulationPosition()
     B=&Multiply_Vector(A=$Engine.Owner.Scale, B=$StackContext.InterpolatedTorusOffset)
    )
    Attachment Point Velocity:Vector3f=<0 0 2>
    Dampening Coefficient=2
    Force Strength:NiagaraFloat=&Multiply_Float(
     A=3
     B=&FloatFromCurve(FloatCurve=DataInterface() {propertyValues {Curve {keys [<at 0.2 1>,<at 0.45 1>]}}})
    )
    Spring Tightness=22
   )
   &InheritSourceMovement(
    Applied Position Scale=&MakeVector(
     X=1
     Y=1
     Z=&Multiply_Float(
      A=&Lerp_Float(A=1, B=0.9, Alpha=$User.ParaGlideLeanAlpha)
      B=&FloatFromCurve(FloatCurve=DataInterface() {propertyValues {Curve {keys [{arriveTangent=-3.3047, interpMode=RCIM_Cubic, leaveTangent=-3.3047, tangentMode=RCTM_User, time=0, value=1.5}, <at 0.2 1 RCIM_Cubic>]}}})
     )
    )
    ApplyForce=false
    ApplyPosition=true
    ApplyPositionFalloff=false
   )
   &self:SphereBlockingVolume_01(Sphere Origin:Vector3f=&SimulationPosition(), Sphere Radius=70) {module=SphereBlockingVolume_01}
   &CurlNoiseForce(Noise Strength=11, Pan Noise=true, Pan Noise Field=<0 0 5>)
   /SetVariables(Particles.Velocity=&Add_Vector(
     A=$Particles.Velocity
     B=&MakeVector(
      X=0
      Y=0
      Z=&FloatFromCurve(
       FloatCurve=DataInterface() {propertyValues {Curve {keys [
           {arriveTangent=7.3676, time=0, value=0.1}
           {arriveTangent=2.8333, leaveTangent=2.8333, time=0.08, value=0.92}
           {arriveTangent=0.0699, interpMode=RCIM_Cubic, leaveTangent=0.0699, tangentMode=RCTM_User, time=0.3, value=0.95}
           <at 1 1 RCIM_Cubic>
          ]}}}
       Scale Curve=&RandomRangeFloat(Minimum=-250, Maximum=-350)
      )
     )
    )) {isSetParameters=true, module=SetVariables_#6}
   /SetVariables(Particles.Velocity=&Add_Vector(
     A=$Particles.Velocity
     B=&MakeVector(
      X=0
      Y=0
      Z=&FloatFromCurve(
       FloatCurve=DataInterface() {propertyValues {Curve {keys [<at 0.7 0>,<at 1 1>]}}}
       Scale Curve=&RandomRangeFloat(Minimum=180, Maximum=250)
      )
     )
    )) {isSetParameters=true, module=SetVariables_#7}
   &AccelerationForce(Acceleration=&Multiply_VectorByFloat(
     Vector=&TransformVector(Vector=<0 1 0>, Destination Space=::World)
     Float=&FloatFromCurve(
      FloatCurve=DataInterface() {propertyValues {Curve {keys [{interpMode=RCIM_Cubic, tangentMode=RCTM_User, time=0.2, value=1.1}, {interpMode=RCIM_Cubic, tangentMode=RCTM_User, time=0.8, value=-1}]}}}
      Scale Curve=&RandomRangeFloat(Minimum=2000, Maximum=2500)
     )
    )) {module=AccelerationForce002}
   &SolveForcesAndVelocity(Clamp Velocity=false, Limit Acceleration=false)
   &UpdateMeshOrientation(
    Rotation Rate=&Multiply_Float(
     A=&Divide_Float(A=5, B=$Particles.Mass)
     B=&RandomRangeFloat(Minimum=0.3, Maximum=0.5, Evaluation Type=::NewEnumerator1)
    )
    Rotation Vector=&RandomVector(Vector Scale=1)
   )
   &ScaleMeshSize(Scale Factor=&Multiply_VectorByFloat(
     Vector=$Engine.Owner.Scale
     Float=&FloatFromCurve(FloatCurve=DataInterface() {propertyValues {Curve {keys [
          <at 0 0 RCIM_Cubic>
          <at 0.2 0.5 RCIM_Cubic>
          {arriveTangent=-0.2898, interpMode=RCIM_Cubic, leaveTangent=-0.2898, tangentMode=RCTM_User, time=0.9, value=0.2}
          <at 1 0>
         ]}}})
    ))
  ]
  rendererClasses []
  renderers []
 }
 Snow_Guide {
  emitterUpdateScript [
   &EmitterState(
    Inactive Response=::NewEnumerator2
    Life Cycle Mode=::NewEnumerator1
    Loop Behavior=::NewEnumerator1
    Loop Duration Mode=::NewEnumerator1
    Scalability Mode=::NewEnumerator1
    UseLoopDelay=false
   )
   &SpawnRate(
    Spawn Probability=0.4
    SpawnRate=&FN_FrontEndFloatModifier(Gameplay=1, Front End=0.6)
    Use Spawn Probability=false
   )
  ]
  particleSpawnScript [
   &InitializeParticle(
    Color=&MultiplyLinearColor(
     Linear Color=<rgba 0.0075 0.0245 0.05 1>
     Scale Factor=&MakeVector4_FromFloat(Float=&FN_FrontEndFloatModifier(Gameplay=1, Front End=2.222))
    )
    Mass Mode=::NewEnumerator1
    UsePositionOffset=false
    Write Material Random=true
   )
   &CalculateSizeAndRotationalInertiaByMass(Density by Material Type:ENiagaraMassByVolume=::NewEnumerator2, "Manually Entered Density (kg/m^3)"=10, Module.Manually Enter Density=false)
   &ShapeLocation(
    MaskSpawnGroup=false
    Non Uniform Scale=<1.1 -2 1>
    Offset=&Lerp_Vector(A=<0 -180 300>, B=<0 -180 450>, Alpha=$User.ParaGlideLeanAlpha)
    Offset Coordinate Space=::Local
    Offset Mode=::NewEnumerator0
    Override Seed=false
    Random Seed=625
    Shape Primitive=::NewEnumerator4
    U Distribution=0.5
    UseOwnerScale=true
   ) {module=ShapeLocation001}
   &ApplyInitialForces(Apply Force to Position=true)
  ]
  particleUpdateScript [
   &ParticleState()
   &Drag(
    Drag=&Multiply_Float(
     A=20
     B=&FloatFromCurve(FloatCurve=DataInterface() {propertyValues {Curve {keys [{interpMode=RCIM_Cubic, tangentMode=RCTM_User, time=0, value=0}, {interpMode=RCIM_Cubic, tangentMode=RCTM_User, time=0.5, value=1}]}}})
    )
    Use Linear Drag=true
    Use Rotational Drag=false
   )
   &InheritSourceMovement(
    Applied Position Scale=&MakeVector(
     X=1
     Y=1
     Z=&Multiply_Float(
      A=&Lerp_Float(A=1, B=1.03, Alpha=$User.ParaGlideLeanAlpha)
      B=&Lerp_Float(
       A=&FloatFromCurve(FloatCurve=DataInterface() {propertyValues {Curve {keys [{arriveTangent=-3.3047, interpMode=RCIM_Cubic, leaveTangent=-3.3047, tangentMode=RCTM_User, time=0, value=1.6}, <at 0.2 1 RCIM_Cubic>]}}})
       B=&FloatFromCurve(FloatCurve=DataInterface() {propertyValues {Curve {keys [{arriveTangent=-3.3047, interpMode=RCIM_Cubic, leaveTangent=-3.3047, tangentMode=RCTM_User, time=0, value=1.3}, <at 0.2 1 RCIM_Cubic>]}}})
       Alpha=$User.ParaGlideLeanAlpha
      )
     )
    )
    ApplyForce=false
    ApplyPosition=true
    ApplyPositionFalloff=false
   )
   &self:SphereBlockingVolume_03(Sphere Origin:Vector3f=&SimulationPosition(), Sphere Radius=70) {module=SphereBlockingVolume_03}
   &CurlNoiseForce(Noise Strength=5, Pan Noise=true, Pan Noise Field=<0 0 5>) {enabled=false}
   &VortexForce(
    UseOriginOffset=false
    Vortex Force Amount=&Multiply_Float(A=&RandomRangeFloat(Minimum=2400, Maximum=2700), B=$System.RotationalSpeed)
   )
   /SetVariables(Particles.Velocity=&Add_Vector(
     A=$Particles.Velocity
     B=&MakeVector(
      X=0
      Y=0
      Z=&FloatFromCurve(
       FloatCurve=DataInterface() {propertyValues {Curve {keys [
           {arriveTangent=7.3676, time=0, value=0.1}
           {arriveTangent=2.8333, leaveTangent=2.8333, time=0.08, value=0.92}
           {arriveTangent=0.0699, interpMode=RCIM_Cubic, leaveTangent=0.0699, tangentMode=RCTM_User, time=0.3, value=0.95}
           <at 1 1 RCIM_Cubic>
          ]}}}
       Scale Curve=&Multiply_Float(
        A=&RandomRangeFloat(Minimum=-80, Maximum=-100)
        B=&FN_FrontEndFloatModifier(Gameplay=1, Front End=&Lerp_Float(A=1.5, B=2, Alpha=$User.ParaGlideLeanAlpha))
       )
      )
     )
    )) {isSetParameters=true, module=SetVariables_#6}
   /SetVariables(Particles.Velocity=&Add_Vector(
     A=$Particles.Velocity
     B=&MakeVector(
      X=0
      Y=0
      Z=&FloatFromCurve(
       FloatCurve=DataInterface() {propertyValues {Curve {keys [<at 0.7 0>,<at 1 1>]}}}
       Scale Curve=&RandomRangeFloat(Minimum=180, Maximum=250)
      )
     )
    )) {isSetParameters=true, module=SetVariables_#7}
   &AccelerationForce(Acceleration=&Multiply_VectorByFloat(
     Vector=&TransformVector(Vector=<0 1 0>, Destination Space=::World)
     Float=&FloatFromCurve(
      FloatCurve=DataInterface() {propertyValues {Curve {keys [{interpMode=RCIM_Cubic, tangentMode=RCTM_User, time=0.2, value=1}, {interpMode=RCIM_Cubic, tangentMode=RCTM_User, time=0.8, value=-1}]}}}
      Scale Curve=&RandomRangeFloat(Minimum=2000, Maximum=2500)
     )
    )) {module=AccelerationForce002}
   &SolveForcesAndVelocity(Clamp Velocity=true, Limit Acceleration=true)
  ]
  rendererClasses [&NiagaraSpriteRendererProperties]
  renderers [{rendererClass=&NiagaraSpriteRendererProperties}]
 }
 Snow_Guide_Ribbon {
  emitterUpdateScript [
   &EmitterState(Scalability Mode=::NewEnumerator1)
   &SpawnParticlesFromOtherEmitter(
    Attribute Reader=DataInterface() {propertyValues {EmitterBinding {bindingMode=Other, emitterName=Snow_Guide}}}
    Spawn Rate=&Lerp_Float(A=100, B=300, Alpha=$User.ParaGlideLeanAlpha)
    Spawn Rate Per Particle Cap=10000000
    Use Maximum Age=false
    Use Minimum Age=false
   )
  ]
  particleSpawnScript [
   &InitializeParticle(Color=<rgba 1 1 1 0.7>, Lifetime=0.3, UsePositionOffset=false, Write Material Random=true)
   &InitializeRibbon(
    Lifetime=0.3
    Ribbon Facing Mode:ENiagara_UnsetDirectSet=::NewEnumerator1
    Ribbon Facing Vector:Vector3f=<1 0 0>
    Ribbon Width=40
    UsePositionOffset=false
    Write Material Random=true
   )
   &SampleParticlesFromOtherEmitter(
    Applied Velocity Scale=-0.1
    Particle ID Sampling=::NewEnumerator1
    Velocity Application:ENiagara_DirectReadApplicationMode=::NewEnumerator0
    Velocity Sampling=::NewEnumerator0
   )
   /SetVariables(Particles.Color=&MultiplyLinearColor(
     Linear Color=$Particles.Initial.Color
     Scale Factor=&MakeVector4_FromLinearColor(LinearColor=&FN_Select_Color_from_Time_Of_Day(Morning=<rgba 0.7605 0.491 0.6172 1>, Day=<rgba 1 1 1 1>, Evening=<rgba 0.7605 0.491 0.6172 1>, Night=<rgba 0.6516 0.7281 0.8177 1>))
    )) {isSetParameters=true, module=SetVariables_#9}
   &AddVelocity(Velocity=<0 0 200>)
  ]
  particleUpdateScript [
   &InheritSourceMovement(ApplyForce=false, ApplyPosition=true, ApplyPositionFalloff=false)
   &SolveForcesAndVelocity(Clamp Velocity=false, Limit Acceleration=false)
   &ParticleState()
   &self:ReadRibbonSourceParticleAge_02(PAR=DataInterface() {propertyValues {EmitterBinding {bindingMode=Other, emitterName=Snow_Guide}}}) {module=ReadRibbonSourceParticleAge_02}
   &ScaleColor(
    Scale Alpha=&FloatFromCurve(FloatCurve=DataInterface() {propertyValues {Curve {keys [<at 0.6 1>,<at 1 0>]}}}, CurveIndex=$Particles.SourceParticleAge)
    Scale RGB=&Lerp_Vector(
     A=&MakeVectorFromLinearColorRGB(Color=<rgba 0.3863 0.6754 0.9271 1>)
     B=&MakeVectorFromLinearColorRGB(Color=<rgba 0.6867 0.7454 0.8148 1>)
     Alpha=&FloatFromCurve(FloatCurve=DataInterface() {propertyValues {Curve {keys [<at 0.7 0 RCIM_Cubic>,<at 1 1 RCIM_Cubic>]}}})
    )
    ScaleA=true
    ScaleRGB=true
   )
   &ScaleRibbonWidth(Ribbon Width Scale=&FloatFromCurve(FloatCurve=DataInterface() {propertyValues {Curve {keys [<at 0.6 1>,<at 1 0.2>]}}}, CurveIndex=$Particles.SourceParticleAge))
   &ScaleRibbonWidth(Ribbon Width Scale=&FloatFromCurve(FloatCurve=DataInterface() {propertyValues {Curve {keys [{tangentMode=RCTM_User, time=0, value=0.4}, {tangentMode=RCTM_User, time=1, value=1}]}}})) {module=ScaleRibbonWidth001}
   &ScaleColor(
    Scale Alpha=&FloatFromCurve(FloatCurve=DataInterface() {propertyValues {Curve {keys [
         <at 0 1 RCIM_Cubic>
         {arriveTangent=-1.4106, interpMode=RCIM_Cubic, leaveTangent=-1.4106, tangentMode=RCTM_User, time=0.5, value=0.3}
         <at 0.85 0 RCIM_Cubic>
        ]}}})
    ScaleA=true
    ScaleRGB=true
   ) {module=ScaleColor001}
  ]
  rendererClasses [&NiagaraRibbonRendererProperties]
  renderers [{rendererClass=&NiagaraRibbonRendererProperties}]
 }
 Snow_Guide_old_spring {
  bEnabled=false
  emitterUpdateScript [
   &EmitterState(
    Inactive Response=::NewEnumerator2
    Life Cycle Mode=::NewEnumerator1
    Loop Behavior=::NewEnumerator1
    Loop Duration Mode=::NewEnumerator1
    UseLoopDelay=false
   )
   &SpawnRate(SpawnRate=3.5, Use Spawn Probability=false)
  ]
  particleSpawnScript [
   &InitializeParticle(
    Color=&MultiplyLinearColor(
     Linear Color=<rgba 0.0075 0.0245 0.05 1>
     Scale Factor=&MakeVector4_FromFloat(Float=&FN_FrontEndFloatModifier(Gameplay=1, Front End=2.222))
    )
    Mass=&Lerp_Float(
     A=1
     B=2
     Alpha=&FloatFromCurve(
      FloatCurve=DataInterface() {propertyValues {Curve {keys [
          {interpMode=RCIM_Cubic, tangentMode=RCTM_User, time=0, value=0}
          {arriveTangent=3.2438, interpMode=RCIM_Cubic, leaveTangent=3.2438, tangentMode=RCTM_User, time=1, value=1}
         ]}}}
      CurveIndex=&RandomRangeFloat(Minimum=0, Maximum=1)
     )
    )
    Mass Mode=::NewEnumerator1
    UsePositionOffset=false
    Write Material Random=true
   )
   &CalculateSizeAndRotationalInertiaByMass(Density by Material Type:ENiagaraMassByVolume=::NewEnumerator2, "Manually Entered Density (kg/m^3)"=10, Module.Manually Enter Density=false)
   &TorusLocation(
    Align Torus With Axis=true
    Coordinate Space:ENiagaraCoordinateSpace=::Simulation
    Handle Radius=40
    MaskSpawnGroup=false
    Module.Override Randomness=false
    Normalized Angle Around Torus Axis=0.2
    Offset=<0 0 300>
    Torus Axis=&RandomRangeVector(Minimum=<0.1 -0.1 1>, Maximum=<-0.1 0.1 1>)
    Torus Origin:NiagaraPosition=&SimulationPosition()
   )
   /SetVariables(
    Particles.InitialTorusOffset:Vector3f=&Add_Vector(A=$Output.TorusLocation.TorusVector, B=<0 0 300>)
    Particles.MeshIndex=&RandomRangeInt(Minimum=0, Maximum=1)
   ) {isSetParameters=true, module=SetVariables_#5}
   &InitialMeshOrientation(Use Rotation Vector=false)
   &ApplyInitialForces(Apply Force to Position=true)
  ]
  particleUpdateScript [
   &ParticleState()
   &InterpolateOverTime(
    Initial Vector Value:Vector3f=$Particles.Initial.InitialTorusOffset
    Rate Of Change=35
    Target Vector Value:Vector3f=&TransformVector(Vector=$Particles.Initial.InitialTorusOffset, Destination Space=::Simulation, Apply Scale=true)
    Variable Type=::NewEnumerator2
   )
   /SetVariables(Particles.InterpolatedTorusOffset:Vector3f=&Multiply_VectorByFloat(
     Vector=&NormalizeVector(
      Vector To Normalize=$StackContext.InterpolateOverTime.InterpolateOverTime_V3_.Moving Average
      Safe Normalize=true
      Safe Fallback Threshold=0
      Safe Fallback Vector=<0 0 0.001>
     )
     Float=&VectorLength(Vector=$Particles.Initial.InitialTorusOffset)
    )) {isSetParameters=true, module=SetVariables_#8}
   &ScaleMeshSize(Scale Factor=&Multiply_VectorByFloat(
     Vector=$Engine.Owner.Scale
     Float=&FloatFromCurve(FloatCurve=DataInterface() {propertyValues {Curve {keys [
          <at 0 0 RCIM_Cubic>
          <at 0.2 0.5 RCIM_Cubic>
          {arriveTangent=-0.2898, interpMode=RCIM_Cubic, leaveTangent=-0.2898, tangentMode=RCTM_User, time=0.9, value=0.2}
          <at 1 0>
         ]}}})
    ))
   &Drag(
    Drag=&Multiply_Float(
     A=20
     B=&FloatFromCurve(FloatCurve=DataInterface() {propertyValues {Curve {keys [{interpMode=RCIM_Cubic, tangentMode=RCTM_User, time=0, value=-0.1}, {interpMode=RCIM_Cubic, tangentMode=RCTM_User, time=0.5, value=1}]}}})
    )
    Use Linear Drag=true
    Use Rotational Drag=false
   )
   &SpringForce(
    " Particle Equilibrium Position":NiagaraPosition=&Add_Vector(
     A=&SimulationPosition()
     B=&Multiply_Vector(A=$Engine.Owner.Scale, B=$StackContext.InterpolatedTorusOffset)
    )
    Attachment Point Velocity:Vector3f=<0 0 2>
    Dampening Coefficient=2
    Force Strength:NiagaraFloat=&Multiply_Float(
     A=3
     B=&FloatFromCurve(FloatCurve=DataInterface() {propertyValues {Curve {keys [<at 0 1>,<at 0.45 1>]}}})
    )
    Spring Tightness=22
   )
   &InheritSourceMovement(
    Applied Position Scale=&MakeVector(
     X=1
     Y=1
     Z=&Multiply_Float(
      A=&Lerp_Float(A=1, B=1.1, Alpha=$User.ParaGlideLeanAlpha)
      B=&FloatFromCurve(FloatCurve=DataInterface() {propertyValues {Curve {keys [{arriveTangent=-3.3047, interpMode=RCIM_Cubic, leaveTangent=-3.3047, tangentMode=RCTM_User, time=0, value=1.5}, <at 0.2 1 RCIM_Cubic>]}}})
     )
    )
    ApplyForce=false
    ApplyPosition=true
    ApplyPositionFalloff=false
   )
   &self:SphereBlockingVolume_01(Sphere Origin:Vector3f=&SimulationPosition(), Sphere Radius=70) {module=SphereBlockingVolume_01}
   &CurlNoiseForce(Noise Strength=11, Pan Noise=true, Pan Noise Field=<0 0 5>)
   /SetVariables(Particles.Velocity=&Add_Vector(
     A=$Particles.Velocity
     B=&MakeVector(
      X=0
      Y=0
      Z=&FloatFromCurve(
       FloatCurve=DataInterface() {propertyValues {Curve {keys [
           {arriveTangent=7.3676, time=0, value=0.1}
           {arriveTangent=2.8333, leaveTangent=2.8333, time=0.08, value=0.92}
           {arriveTangent=0.0699, interpMode=RCIM_Cubic, leaveTangent=0.0699, tangentMode=RCTM_User, time=0.3, value=0.95}
           <at 1 1 RCIM_Cubic>
          ]}}}
       Scale Curve=&RandomRangeFloat(Minimum=-200, Maximum=-300)
      )
     )
    )) {isSetParameters=true, module=SetVariables_#6}
   /SetVariables(Particles.Velocity=&Add_Vector(
     A=$Particles.Velocity
     B=&MakeVector(
      X=0
      Y=0
      Z=&FloatFromCurve(
       FloatCurve=DataInterface() {propertyValues {Curve {keys [<at 0.7 0>,<at 1 1>]}}}
       Scale Curve=&RandomRangeFloat(Minimum=180, Maximum=250)
      )
     )
    )) {isSetParameters=true, module=SetVariables_#7}
   &AccelerationForce(Acceleration=&Multiply_VectorByFloat(
     Vector=&TransformVector(Vector=<0 1 0>, Destination Space=::World)
     Float=&FloatFromCurve(
      FloatCurve=DataInterface() {propertyValues {Curve {keys [{interpMode=RCIM_Cubic, tangentMode=RCTM_User, time=0.2, value=1.1}, {interpMode=RCIM_Cubic, tangentMode=RCTM_User, time=0.8, value=-1}]}}}
      Scale Curve=&RandomRangeFloat(Minimum=2000, Maximum=2500)
     )
    )) {module=AccelerationForce002}
   &SolveForcesAndVelocity(Clamp Velocity=true, Limit Acceleration=true)
   &UpdateMeshOrientation(
    Rotation Rate=&Multiply_Float(
     A=&Divide_Float(A=5, B=$Particles.Mass)
     B=&RandomRangeFloat(Minimum=0.3, Maximum=0.5, Evaluation Type=::NewEnumerator1)
    )
    Rotation Vector=&RandomVector(Vector Scale=1)
   )
  ]
  rendererClasses []
  renderers []
 }
 Snow_Ribbon {
  emitterUpdateScript [
   &EmitterState(
    Delay First Loop Only=true
    Life Cycle Mode=::NewEnumerator1
    Loop Delay=0.01
    Loop Duration=1
    Scalability Mode=::NewEnumerator1
    UseLoopDelay=true
   )
   &SpawnParticlesFromOtherEmitter(
    Attribute Reader=DataInterface() {propertyValues {EmitterBinding {bindingMode=Other, emitterName=Snow_Chunks}}}
    Spawn Rate=&Lerp_Float(A=100, B=300, Alpha=$User.ParaGlideLeanAlpha)
    Spawn Rate Per Particle Cap=10000000
    Use Maximum Age=false
    Use Minimum Age=false
   )
  ]
  particleSpawnScript [
   &InitializeParticle(Color=<rgba 1 1 1 0.7>, Lifetime=0.3, UsePositionOffset=false, Write Material Random=true)
   &InitializeRibbon(
    Lifetime=0.3
    Ribbon Facing Mode:ENiagara_UnsetDirectSet=::NewEnumerator1
    Ribbon Facing Vector:Vector3f=<1 0 0>
    Ribbon Width=40
    UsePositionOffset=false
    Write Material Random=true
   )
   &SampleParticlesFromOtherEmitter(
    Applied Velocity Scale=0
    Particle ID Sampling=::NewEnumerator1
    Velocity Application:ENiagara_DirectReadApplicationMode=::NewEnumerator0
    Velocity Sampling=::NewEnumerator0
   )
   /SetVariables(Particles.Color=&MultiplyLinearColor(
     Linear Color=$Particles.Initial.Color
     Scale Factor=&MakeVector4_FromLinearColor(LinearColor=&FN_Select_Color_from_Time_Of_Day(Morning=<rgba 1 1 1 1>, Day=<rgba 1 1 1 1>, Evening=<rgba 1 1 1 1>, Night=<rgba 1 1 1 1>))
    )) {enabled=false, isSetParameters=true, module=SetVariables_#9}
   &AddVelocity(Velocity=<0 0 200>)
  ]
  particleUpdateScript [
   &InheritSourceMovement(ApplyForce=false, ApplyPosition=true, ApplyPositionFalloff=false)
   &SolveForcesAndVelocity(Clamp Velocity=false, Limit Acceleration=false)
   &ParticleState()
   &self:ReadRibbonSourceParticleAge_01(PAR=DataInterface() {propertyValues {EmitterBinding {bindingMode=Other, emitterName=Snow_Chunks}}}) {module=ReadRibbonSourceParticleAge_01}
   &ScaleColor(
    Scale Alpha=&FloatFromCurve(
     FloatCurve=DataInterface() {propertyValues {Curve {keys [<at 0.6 1 RCIM_Cubic>,<at 1 0 RCIM_Cubic>]}}}
     CurveIndex=$Particles.SourceParticleAge
    )
    Scale RGB=&Lerp_Vector(
     A=&MakeVectorFromLinearColorRGB(Color=<rgba 0.3864 0.6724 0.9301 1>)
     B=&MakeVectorFromLinearColorRGB(Color=<rgba 0.6867 0.7454 0.8148 1>)
     Alpha=&FloatFromCurve(FloatCurve=DataInterface() {propertyValues {Curve {keys [<at 0.7 0 RCIM_Cubic>,<at 1 1 RCIM_Cubic>]}}})
    )
    ScaleA=true
    ScaleRGB=true
   )
   &ScaleRibbonWidth(Ribbon Width Scale=&FloatFromCurve(FloatCurve=DataInterface() {propertyValues {Curve {keys [<at 0.6 1>,<at 1 0.2>]}}}, CurveIndex=$Particles.SourceParticleAge))
   &ScaleRibbonWidth(Ribbon Width Scale=&FloatFromCurve(FloatCurve=DataInterface() {propertyValues {Curve {keys [{tangentMode=RCTM_User, time=0, value=0.4}, {tangentMode=RCTM_User, time=1, value=1}]}}})) {module=ScaleRibbonWidth001}
   &ScaleColor(
    Scale Alpha=&FloatFromCurve(FloatCurve=DataInterface() {propertyValues {Curve {keys [
         <at 0 1 RCIM_Cubic>
         {arriveTangent=-0.6811, interpMode=RCIM_Cubic, leaveTangent=-0.6811, tangentMode=RCTM_User, time=0.5, value=0.1}
         <at 0.75 0 RCIM_Cubic>
        ]}}})
    ScaleA=true
    ScaleRGB=true
   ) {module=ScaleColor001}
  ]
  rendererClasses [&NiagaraRibbonRendererProperties]
  renderers [{rendererClass=&NiagaraRibbonRendererProperties}]
 }
 Snowflakes {
  emitterUpdateScript [
   &EmitterState()
   &SpawnParticlesFromOtherEmitter(
    Attribute Reader=DataInterface() {propertyValues {EmitterBinding {bindingMode=Other, emitterName=Snow_Chunks}}}
    Spawn Rate=28
    Spawn Rate Per Particle Cap=28
    Use Maximum Age=false
    Use Minimum Age=false
   )
  ]
  particleSpawnScript [
   &InitializeParticle(
    Color=<rgba 0.6048 0.6981 0.8177 1>
    Lifetime=&RandomRangeFloat(Minimum=0.5, Maximum=0.9)
    Sprite Rotation Mode=::NewEnumerator2
    Sprite Size Mode=::NewEnumerator3
    Uniform Sprite Size=&Multiply_Float(
     A=&RandomRangeFloat(Minimum=8, Maximum=10)
     B=&FN_FrontEndFloatModifier(Gameplay=1, Front End=0.7)
    )
    UsePositionOffset=false
    Write Material Random=true
   )
   &SampleParticlesFromOtherEmitter(Applied Velocity Scale=-0.1, Velocity Sampling=::NewEnumerator0)
   &AddVelocity(Velocity=<0 0 400>)
   /SetVariables(Particles.Color=&MultiplyLinearColor(
     Linear Color=$Particles.Initial.Color
     Scale Factor=&MakeVector4_FromLinearColor(LinearColor=&FN_Select_Color_from_Time_Of_Day(Morning=<rgba 0.7605 0.491 0.6172 1>, Day=<rgba 1 1 1 1>, Evening=<rgba 0.7605 0.491 0.6172 1>, Night=<rgba 0.6516 0.7281 0.8177 1>))
    )) {isSetParameters=true}
   &SphereLocation(
    MaskSpawnGroup=false
    Offset=<0 0 15>
    Override Seed=false
    Sphere Radius=18
    Surface Only=false
    UseNonUniformScale=false
    UseOffset=true
    UseSphereOrientation=false
   )
   &InitialMeshOrientation(
    Mesh Orientation Mode=::NewEnumerator2
    Rotation=&RandomRangeVector(Minimum=<0 0 0>, Maximum=<1 0 0>)
    Use Rotation Vector=true
   ) {module=InitialMeshOrientation001}
   /Niagara/Modules/Update/SubUV/SubUVAnimation(SubUV Lookup Index=&UniformRangedFloat(Minimum=0, Maximum=3))
  ]
  particleUpdateScript [
   &InheritSourceMovement(ApplyForce=false, ApplyPosition=true, ApplyPositionFalloff=false)
   &SolveForcesAndVelocity(Clamp Velocity=false, Limit Acceleration=false)
   &ParticleState()
   &ScaleColor(
    Scale Alpha=&Multiply_Float(
     A=&FloatFromCurve(
      FloatCurve=DataInterface() {propertyValues {Curve {keys [
          {interpMode=RCIM_Cubic, tangentMode=RCTM_User, time=0, value=1}
          {interpMode=RCIM_Cubic, leaveTangent=-2.6773, tangentMode=RCTM_Break, time=0.3, value=1}
          {tangentMode=RCTM_User, time=1, value=0}
         ]}}}
      Scale Curve=&RandomRangeFloat(Minimum=0.5, Maximum=1)
     )
     B=&FNTimeOfDayScalar(Night Scale=0.6, Morning Scale=0.4, Evening Scale=1)
    )
    ScaleA=true
    ScaleRGB=true
   )
   &ScaleSpriteSize(Uniform Curve Sprite Scale=DataInterface() {propertyValues {Curve {keys [
        {arriveTangent=9.0725, interpMode=RCIM_Cubic, leaveTangent=9.0725, tangentMode=RCTM_User, time=0, value=0}
        {arriveTangent=-1.1228, interpMode=RCIM_Cubic, leaveTangent=-1.1228, tangentMode=RCTM_User, time=0.2, value=1}
        {arriveTangent=-0.0405, interpMode=RCIM_Cubic, leaveTangent=-0.0405, tangentMode=RCTM_User, time=1, value=0.2}
       ]}}})
  ]
  rendererClasses [&NiagaraSpriteRendererProperties]
  renderers [{rendererClass=&NiagaraSpriteRendererProperties}]
 }
 Snowflakes_Guide {
  emitterUpdateScript [
   &EmitterState()
   &SpawnParticlesFromOtherEmitter(
    Attribute Reader=DataInterface() {propertyValues {EmitterBinding {bindingMode=Other, emitterName=Snow_Guide}}}
    Spawn Rate=28
    Spawn Rate Per Particle Cap=28
    Use Maximum Age=false
    Use Minimum Age=false
   )
  ]
  particleSpawnScript [
   &InitializeParticle(
    Color=<rgba 0.6048 0.6981 0.8177 0.8>
    Lifetime=&RandomRangeFloat(Minimum=0.5, Maximum=0.9)
    Sprite Rotation Mode=::NewEnumerator2
    Sprite Size Mode=::NewEnumerator3
    Uniform Sprite Size=&Multiply_Float(
     A=&RandomRangeFloat(Minimum=8, Maximum=10)
     B=&FN_FrontEndFloatModifier(Gameplay=1, Front End=0.7)
    )
    UsePositionOffset=false
    Write Material Random=true
   )
   &SampleParticlesFromOtherEmitter(Applied Velocity Scale=-0.1, Velocity Sampling=::NewEnumerator0)
   &AddVelocity(Velocity=<0 0 300>)
   /SetVariables(Particles.Color=&MultiplyLinearColor(
     Linear Color=$Particles.Initial.Color
     Scale Factor=&MakeVector4_FromLinearColor(LinearColor=&FN_Select_Color_from_Time_Of_Day(Morning=<rgba 0.7605 0.491 0.6172 1>, Day=<rgba 1 1 1 1>, Evening=<rgba 0.7605 0.491 0.6172 1>, Night=<rgba 0.6516 0.7281 0.8177 1>))
    )) {isSetParameters=true}
   &SphereLocation(
    MaskSpawnGroup=false
    Offset=<0 0 15>
    Override Seed=false
    Sphere Radius=20
    Surface Only=false
    UseNonUniformScale=false
    UseOffset=true
    UseSphereOrientation=false
   )
   &InitialMeshOrientation(
    Mesh Orientation Mode=::NewEnumerator2
    Rotation=&RandomRangeVector(Minimum=<0 0 0>, Maximum=<1 0 0>)
    Use Rotation Vector=true
   ) {module=InitialMeshOrientation001}
   &SubUVAnimation(SubUV Lookup Index=&UniformRangedFloat(Minimum=0, Maximum=3))
  ]
  particleUpdateScript [
   &InheritSourceMovement(ApplyForce=false, ApplyPosition=true, ApplyPositionFalloff=false)
   &SolveForcesAndVelocity(Clamp Velocity=false, Limit Acceleration=false)
   &ParticleState()
   &ScaleColor(
    Scale Alpha=&Multiply_Float(
     A=&FloatFromCurve(
      FloatCurve=DataInterface() {propertyValues {Curve {keys [
          {interpMode=RCIM_Cubic, tangentMode=RCTM_User, time=0, value=1}
          {interpMode=RCIM_Cubic, leaveTangent=-2.6773, tangentMode=RCTM_Break, time=0.3, value=1}
          {tangentMode=RCTM_User, time=1, value=0}
         ]}}}
      Scale Curve=&RandomRangeFloat(Minimum=0.5, Maximum=1)
     )
     B=&FNTimeOfDayScalar(Night Scale=0.6, Morning Scale=0.4, Evening Scale=0.8)
    )
    ScaleA=true
    ScaleRGB=true
   )
   &ScaleSpriteSize(Uniform Curve Sprite Scale=DataInterface() {propertyValues {Curve {keys [
        {arriveTangent=9.0725, interpMode=RCIM_Cubic, leaveTangent=9.0725, tangentMode=RCTM_User, time=0, value=0}
        {arriveTangent=-1.1228, interpMode=RCIM_Cubic, leaveTangent=-1.1228, tangentMode=RCTM_User, time=0.1408, value=0.9847}
        {arriveTangent=-0.0405, interpMode=RCIM_Cubic, leaveTangent=-0.0405, tangentMode=RCTM_User, time=1, value=0.2}
       ]}}})
  ]
  rendererClasses [&NiagaraSpriteRendererProperties]
  renderers [{rendererClass=&NiagaraSpriteRendererProperties}]
 }
 Speed_Bits_CPU {
  emitterUpdateScript [
   &EmitterState()
   &SpawnRate(
    SpawnRate=&Multiply_Float(
     A=170
     B=&Lerp_Float(A=0, B=&FN_FrontEndFloatModifier(Gameplay=1, Front End=0.4), Alpha=$User.ParaGlideLeanAlpha)
    )
    Use Spawn Probability=false
   )
  ]
  particleSpawnScript [
   &InitializeParticle(
    Color=&MultiplyLinearColor(
     Linear Color=<rgba 0.5317 0.6897 0.8438 1>
     Scale Factor=&MakeVector4(X=1, Y=1, Z=1, W=&RandomRangeFloat(Minimum=0.35, Maximum=1))
    )
    Lifetime Max=0.4
    Lifetime Min=0.2
    Lifetime Mode=::NewEnumerator1
    Sprite Rotation Mode=::NewEnumerator2
    Sprite Size=&Multiply_Vector2D_ByFloat(
     Float=&FN_FrontEndFloatModifier(Gameplay=0.4, Front End=&RandomRangeFloat(Minimum=0.4, Maximum=0.6))
     Vector2D=&RandomRangeVector2D(Minimum=<4 30>, Maximum=<13 50>)
    )
    Sprite Size Mode=::NewEnumerator4
    UsePositionOffset=false
    Write Material Random=true
   )
   &SphereLocation(
    MaskSpawnGroup=false
    Non Uniform Scale=<1 1 0.1>
    Offset=<0 0 -40>
    Override Seed=false
    Sphere Radius=40
    Surface Only=false
    UseNonUniformScale=true
    UseOffset=true
    UseSphereOrientation=false
   )
   &AddVelocityFromPoint(
    Module.Use Falloff Distance=false
    Origin Offset=<0 0 -80>
    Velocity Strength=&Multiply_Float(A=600, B=&RandomRangeFloat(Minimum=0.8, Maximum=1))
   )
   &AddVelocityInCone(Cone Angle=30, Cone Axis=<0 0 1>, Module.Use Velocity Falloff On Cone Axis=false, Velocity Strength=600) {module=AddVelocityInCone001}
   &CameraOffset()
   &InitialMeshOrientation(
    Mesh Orientation Mode=::NewEnumerator2
    Rotation=&RandomRangeVector(Minimum=<0 0 0>, Maximum=<1 0 0>)
    Use Rotation Vector=true
   ) {module=InitialMeshOrientation001}
   /Niagara/Modules/Update/SubUV/V2/SubUVAnimation(End Frame=1, SubUV Animation Mode:ENiagara_SubUVLookupMode=::NewEnumerator2)
  ]
  particleUpdateScript [
   &AccelerationForce(Acceleration=&MakeVector(X=0, Y=0, Z=300))
   &Drag(Use Linear Drag=true, Use Rotational Drag=false)
   &VortexForce(
    UseOriginOffset=false
    Vortex Force Amount=&Multiply_Float(A=&RandomRangeFloat(Minimum=1000, Maximum=200), B=$System.RotationalSpeed)
   )
   &SolveForcesAndVelocity(Clamp Velocity=false, Limit Acceleration=false)
   &ParticleState()
   &ScaleSpriteSize(Scale Factor=&Vector2DFromFloat(Value=&FloatFromCurve(FloatCurve=DataInterface() {propertyValues {Curve {keys [
          {arriveTangent=6.4968, interpMode=RCIM_Cubic, time=0, value=0}
          {arriveTangent=1.4286, interpMode=RCIM_Cubic, leaveTangent=1.4286, time=0.1351, value=1}
          {arriveTangent=-1.1563, interpMode=RCIM_Cubic, leaveTangent=-1.1563, time=0.7, value=1}
          {interpMode=RCIM_Cubic, leaveTangent=-1.0152, time=1, value=0}
         ]}}}))) {module=ScaleSpriteSize001}
   &SpriteRotationRate(Rotation Rate=&RandomRangeFloat(Minimum=-1, Maximum=1))
   &ScaleColor(
    Scale Alpha=&Multiply_Float(
     A=&RandomRangeFloat(Minimum=0.2, Maximum=1)
     B=&FloatFromCurve(FloatCurve=DataInterface() {propertyValues {Curve {keys [
          {interpMode=RCIM_Cubic, tangentMode=RCTM_User, time=0, value=0}
          <at 0.3 1>
          {interpMode=RCIM_Cubic, tangentMode=RCTM_User, time=0.5, value=1}
          <at 0.7 1>
          {tangentMode=RCTM_User, time=1, value=0}
         ]}}})
    )
    ScaleA=true
    ScaleRGB=true
   )
  ]
  rendererClasses [&NiagaraSpriteRendererProperties]
  renderers [{rendererClass=&NiagaraSpriteRendererProperties}]
 }
 Speed_Bits_GPU {
  simTarget=GPUComputeSim
  emitterUpdateScript [
   &EmitterState()
   &SpawnRate(
    SpawnRate=&Multiply_Float(
     A=170
     B=&Lerp_Float(A=0, B=&FN_FrontEndFloatModifier(Gameplay=1, Front End=0.4), Alpha=$User.ParaGlideLeanAlpha)
    )
    Use Spawn Probability=false
   )
  ]
  particleSpawnScript [
   &InitializeParticle(
    Color=&MultiplyLinearColor(
     Linear Color=<rgba 0.5317 0.6897 0.8438 1>
     Scale Factor=&MakeVector4(X=1, Y=1, Z=1, W=&RandomRangeFloat(Minimum=0.35, Maximum=1))
    )
    Lifetime Max=0.4
    Lifetime Min=0.2
    Lifetime Mode=::NewEnumerator1
    Sprite Rotation Mode=::NewEnumerator2
    Sprite Size=&Multiply_Vector2D_ByFloat(
     Float=&FN_FrontEndFloatModifier(Gameplay=0.4, Front End=&RandomRangeFloat(Minimum=0.2, Maximum=0.25))
     Vector2D=&RandomRangeVector2D(Minimum=<4 30>, Maximum=<13 50>)
    )
    Sprite Size Mode=::NewEnumerator4
    UsePositionOffset=false
    Write Material Random=true
   )
   &SphereLocation(
    MaskSpawnGroup=false
    Non Uniform Scale=<1 1 0.1>
    Offset=<0 0 -40>
    Override Seed=false
    Sphere Radius=40
    Surface Only=false
    UseNonUniformScale=true
    UseOffset=true
    UseSphereOrientation=false
   )
   &AddVelocityFromPoint(
    Module.Use Falloff Distance=false
    Origin Offset=<0 0 -80>
    Velocity Strength=&Multiply_Float(A=600, B=&RandomRangeFloat(Minimum=0.8, Maximum=1))
   )
   &AddVelocityInCone(Cone Angle=30, Cone Axis=<0 0 1>, Module.Use Velocity Falloff On Cone Axis=false, Velocity Strength=600) {module=AddVelocityInCone001}
   &CameraOffset() {enabled=false}
   &InitialMeshOrientation(
    Mesh Orientation Mode=::NewEnumerator2
    Rotation=&RandomRangeVector(Minimum=<0 0 0>, Maximum=<1 0 0>)
    Use Rotation Vector=true
   ) {module=InitialMeshOrientation001}
   &SubUVAnimation(End Frame=1, SubUV Animation Mode:ENiagara_SubUVLookupMode=::NewEnumerator2)
  ]
  particleUpdateScript [
   &AccelerationForce(Acceleration=&MakeVector(X=0, Y=0, Z=300))
   &Drag(Use Linear Drag=true, Use Rotational Drag=false)
   &VortexForce(
    UseOriginOffset=false
    Vortex Force Amount=&Multiply_Float(A=&RandomRangeFloat(Minimum=1000, Maximum=200), B=$System.RotationalSpeed)
   )
   &SolveForcesAndVelocity(Clamp Velocity=false, Limit Acceleration=false)
   &ParticleState()
   &ScaleSpriteSize(Scale Factor=&Vector2DFromFloat(Value=&FloatFromCurve(FloatCurve=DataInterface() {propertyValues {Curve {keys [
          {arriveTangent=6.4968, interpMode=RCIM_Cubic, time=0, value=0}
          {arriveTangent=1.4286, interpMode=RCIM_Cubic, leaveTangent=1.4286, time=0.1351, value=1}
          {arriveTangent=-1.1563, interpMode=RCIM_Cubic, leaveTangent=-1.1563, time=0.7, value=1}
          {interpMode=RCIM_Cubic, leaveTangent=-1.0152, time=1, value=0}
         ]}}}))) {module=ScaleSpriteSize001}
   &SpriteRotationRate(Rotation Rate=&RandomRangeFloat(Minimum=-1, Maximum=1))
   &ScaleColor(
    Scale Alpha=&Multiply_Float(
     A=&RandomRangeFloat(Minimum=0.2, Maximum=1)
     B=&FloatFromCurve(FloatCurve=DataInterface() {propertyValues {Curve {keys [
          {interpMode=RCIM_Cubic, tangentMode=RCTM_User, time=0, value=0}
          <at 0.3 1>
          {interpMode=RCIM_Cubic, tangentMode=RCTM_User, time=0.5, value=1}
          <at 0.7 1>
          {tangentMode=RCTM_User, time=1, value=0}
         ]}}})
    )
    ScaleA=true
    ScaleRGB=true
   )
  ]
  rendererClasses [&NiagaraSpriteRendererProperties]
  renderers [{rendererClass=&NiagaraSpriteRendererProperties}]
 }
 Speed_Rocks_CPU {
  emitterUpdateScript [
   &EmitterState()
   &SpawnRate(
    SpawnRate=&Multiply_Float(A=15, B=&Lerp_Float(A=0.1, B=1, Alpha=$User.ParaGlideLeanAlpha))
    Use Spawn Probability=false
   )
  ]
  particleSpawnScript [
   &InitializeParticle(
    Color=&MultiplyLinearColor(
     Linear Color=<rgba 0.6302 0.8174 1 1>
     Scale Factor=&MakeVector4_FromFloat(Float=&FN_FrontEndFloatModifier(Gameplay=1, Front End=0.8))
    )
    Lifetime Max=0.5
    Lifetime Min=0.15
    Lifetime Mode=::NewEnumerator1
    Mesh Scale=&Multiply_Vector(
     A=&Multiply_VectorByFloat(
      Vector=&RandomRangeVector(Minimum=<2 0.5 0.5>, Maximum=<4 1 1>)
      Float=&RandomRangeFloat(Minimum=0.02, Maximum=0.005)
     )
     B=&VectorFromFloat(Value=&FN_FrontEndFloatModifier(Gameplay=1, Front End=0.7))
    )
    Mesh Scale Mode=::NewEnumerator4
    Sprite Rotation Mode=::NewEnumerator2
    Sprite Size Mode=::NewEnumerator1
    Uniform Sprite Size Max=20
    Uniform Sprite Size Min=10
    UsePositionOffset=false
    Write Material Random=true
   )
   &SphereLocation(
    MaskSpawnGroup=false
    Non Uniform Scale=<1 1 0.1>
    Offset=<0 0 -40>
    Override Seed=false
    Sphere Radius=40
    Surface Only=false
    UseNonUniformScale=true
    UseOffset=true
    UseSphereOrientation=false
   )
   &AddVelocityFromPoint(
    Module.Use Falloff Distance=false
    Origin Offset=<0 0 -80>
    Velocity Strength=&Multiply_Float(A=600, B=&RandomRangeFloat(Minimum=0.8, Maximum=1))
   )
   &AddVelocityInCone(Cone Angle=30, Cone Axis=<0 0 1>, Module.Use Velocity Falloff On Cone Axis=false, Velocity Strength=600) {module=AddVelocityInCone001}
   &CameraOffset()
   &InitialMeshOrientation(
    Mesh Orientation Mode=::NewEnumerator2
    Rotation=&RandomRangeVector(Minimum=<0 0 0>, Maximum=<1 0 0>)
    Use Rotation Vector=true
   ) {module=InitialMeshOrientation001}
  ]
  particleUpdateScript [
   &AccelerationForce(Acceleration=&MakeVector(X=0, Y=0, Z=300))
   &Drag(Use Linear Drag=true, Use Rotational Drag=false)
   &VortexForce(
    UseOriginOffset=false
    Vortex Force Amount=&Multiply_Float(A=&RandomRangeFloat(Minimum=1000, Maximum=200), B=$System.RotationalSpeed)
   )
   &SolveForcesAndVelocity(Clamp Velocity=false, Limit Acceleration=false)
   &ParticleState()
   &ScaleSpriteSize(Scale Factor=&Vector2DFromFloat(Value=&FloatFromCurve(FloatCurve=DataInterface() {propertyValues {Curve {keys [
          {arriveTangent=6.4968, interpMode=RCIM_Cubic, leaveTangent=6.4968, time=0, value=0}
          {interpMode=RCIM_Cubic, leaveTangent=-2.1899, tangentMode=RCTM_Break, time=0.1351, value=0.878}
          {arriveTangent=-1.0152, interpMode=RCIM_Cubic, leaveTangent=-1.0152, time=1, value=0}
         ]}}}))) {module=ScaleSpriteSize001}
   &SpriteRotationRate(Rotation Rate=&RandomRangeFloat(Minimum=-1, Maximum=1))
   &ScaleColor(
    Scale Alpha=&Multiply_Float(
     A=&RandomRangeFloat(Minimum=0.2, Maximum=1)
     B=&FloatFromCurve(FloatCurve=DataInterface() {propertyValues {Curve {keys [
          {interpMode=RCIM_Cubic, tangentMode=RCTM_User, time=0, value=0}
          <at 0.3 1>
          {interpMode=RCIM_Cubic, tangentMode=RCTM_User, time=0.5, value=1}
          <at 0.7 1>
          {tangentMode=RCTM_User, time=1, value=0}
         ]}}})
    )
    ScaleA=true
    ScaleRGB=true
   )
  ]
  rendererClasses [&NiagaraMeshRendererProperties]
  renderers [{rendererClass=&NiagaraMeshRendererProperties, rendererIndex=0}]
 }
 Speed_Rocks_GPU {
  simTarget=GPUComputeSim
  emitterUpdateScript [
   &EmitterState()
   &SpawnRate(
    SpawnRate=&Multiply_Float(A=15, B=&Lerp_Float(A=0.1, B=1, Alpha=$User.ParaGlideLeanAlpha))
    Use Spawn Probability=false
   )
  ]
  particleSpawnScript [
   &InitializeParticle(
    Color=&MultiplyLinearColor(
     Linear Color=<rgba 0.6302 0.8174 1 1>
     Scale Factor=&MakeVector4_FromFloat(Float=&FN_FrontEndFloatModifier(Gameplay=1, Front End=0.8))
    )
    Lifetime Max=0.5
    Lifetime Min=0.15
    Lifetime Mode=::NewEnumerator1
    Mesh Scale=&Multiply_Vector(
     A=&Multiply_VectorByFloat(
      Vector=&RandomRangeVector(Minimum=<2 0.5 0.5>, Maximum=<4 1 1>)
      Float=&RandomRangeFloat(Minimum=0.02, Maximum=0.005)
     )
     B=&VectorFromFloat(Value=&FN_FrontEndFloatModifier(Gameplay=1, Front End=0.7))
    )
    Mesh Scale Mode=::NewEnumerator4
    Sprite Rotation Mode=::NewEnumerator2
    Sprite Size Mode=::NewEnumerator1
    Uniform Sprite Size Max=20
    Uniform Sprite Size Min=10
    UsePositionOffset=false
    Write Material Random=true
   )
   &SphereLocation(
    MaskSpawnGroup=false
    Non Uniform Scale=<1 1 0.1>
    Offset=<0 0 -40>
    Override Seed=false
    Sphere Radius=40
    Surface Only=false
    UseNonUniformScale=true
    UseOffset=true
    UseSphereOrientation=false
   )
   &AddVelocityFromPoint(
    Module.Use Falloff Distance=false
    Origin Offset=<0 0 -80>
    Velocity Strength=&Multiply_Float(A=600, B=&RandomRangeFloat(Minimum=0.8, Maximum=1))
   )
   &AddVelocityInCone(Cone Angle=30, Cone Axis=<0 0 1>, Module.Use Velocity Falloff On Cone Axis=false, Velocity Strength=600) {module=AddVelocityInCone001}
   &CameraOffset()
   &InitialMeshOrientation(
    Mesh Orientation Mode=::NewEnumerator2
    Rotation=&RandomRangeVector(Minimum=<0 0 0>, Maximum=<1 0 0>)
    Use Rotation Vector=true
   ) {module=InitialMeshOrientation001}
  ]
  particleUpdateScript [
   &AccelerationForce(Acceleration=&MakeVector(X=0, Y=0, Z=300))
   &Drag(Use Linear Drag=true, Use Rotational Drag=false)
   &VortexForce(
    UseOriginOffset=false
    Vortex Force Amount=&Multiply_Float(A=&RandomRangeFloat(Minimum=1000, Maximum=200), B=$System.RotationalSpeed)
   )
   &SolveForcesAndVelocity(Clamp Velocity=false, Limit Acceleration=false)
   &ParticleState()
   &ScaleSpriteSize(Scale Factor=&Vector2DFromFloat(Value=&FloatFromCurve(FloatCurve=DataInterface() {propertyValues {Curve {keys [
          {arriveTangent=6.4968, interpMode=RCIM_Cubic, leaveTangent=6.4968, time=0, value=0}
          {interpMode=RCIM_Cubic, leaveTangent=-2.1899, tangentMode=RCTM_Break, time=0.1351, value=0.878}
          {arriveTangent=-1.0152, interpMode=RCIM_Cubic, leaveTangent=-1.0152, time=1, value=0}
         ]}}}))) {module=ScaleSpriteSize001}
   &SpriteRotationRate(Rotation Rate=&RandomRangeFloat(Minimum=-1, Maximum=1))
   &ScaleColor(
    Scale Alpha=&Multiply_Float(
     A=&RandomRangeFloat(Minimum=0.2, Maximum=1)
     B=&FloatFromCurve(FloatCurve=DataInterface() {propertyValues {Curve {keys [
          {interpMode=RCIM_Cubic, tangentMode=RCTM_User, time=0, value=0}
          <at 0.3 1>
          {interpMode=RCIM_Cubic, tangentMode=RCTM_User, time=0.5, value=1}
          <at 0.7 1>
          {tangentMode=RCTM_User, time=1, value=0}
         ]}}})
    )
    ScaleA=true
    ScaleRGB=true
   )
  ]
  rendererClasses [&NiagaraMeshRendererProperties]
  renderers [{rendererClass=&NiagaraMeshRendererProperties, rendererIndex=0}]
 }
}
summary {
 systemName=NS_Contrail_Avalanche_01
 emitters [
  {bEnabled=false, emitterName=Snow_Chunks_old_spring, rendererClasses []}
  {bEnabled=false, emitterName=Snow_Guide_old_spring, rendererClasses []}
  {emitterName=Snow_Chunks, rendererClasses [&NiagaraMeshRendererProperties]}
  {emitterName=Fragments_Chunks, rendererClasses [&NiagaraMeshRendererProperties]}
  {emitterName=Puff_Snow, rendererClasses [&NiagaraSpriteRendererProperties]}
  {emitterName=Snowflakes, rendererClasses [&NiagaraSpriteRendererProperties]}
  {emitterName=Snow_Ribbon, rendererClasses [&NiagaraRibbonRendererProperties]}
  {emitterName=Smoke001, rendererClasses [&NiagaraSpriteRendererProperties]}
  {emitterName=Fragments_Guide, rendererClasses [&NiagaraMeshRendererProperties]}
  {emitterName=Snow_Guide, rendererClasses [&NiagaraSpriteRendererProperties]}
  {emitterName=Snowflakes_Guide, rendererClasses [&NiagaraSpriteRendererProperties]}
  {emitterName=Puff_Snow_Guide, rendererClasses [&NiagaraSpriteRendererProperties]}
  {emitterName=Smoke, rendererClasses [&NiagaraSpriteRendererProperties]}
  {emitterName=Snow_Guide_Ribbon, rendererClasses [&NiagaraRibbonRendererProperties]}
  {emitterName=Speed_Rocks_GPU, simTarget=GPUComputeSim, rendererClasses [&NiagaraMeshRendererProperties]}
  {emitterName=Speed_Bits_GPU, simTarget=GPUComputeSim, rendererClasses [&NiagaraSpriteRendererProperties]}
  {emitterName=Speed_Rocks_CPU, rendererClasses [&NiagaraMeshRendererProperties]}
  {emitterName=Speed_Bits_CPU, rendererClasses [&NiagaraSpriteRendererProperties]}
 ]
 userVariables [
  User.bIsFrontEnd {type=NiagaraBool}
  User.bIsFrontEndPreview {type=NiagaraBool}
  User.FeetRestingPositions {
   type=NiagaraDataInterfaceArrayFloat3
   description=""
   defaultValue {
    struct=NiagaraExt_VariableValue_DataInterface
    value {
     dataInterface=&self:NiagaraDataInterfaceArrayFloat3_0
     dataInterfaceClass=&NiagaraDataInterfaceArrayFloat3
    }
   }
  }
  User.ForwardDot {type=NiagaraFloat, defaultValue=0, description=""}
  User.Landing {type=NiagaraBool}
  User.ParaGlideLeanAlpha {type=NiagaraFloat, defaultValue=0, description=""}
  User.PlayerSpeed {type=NiagaraFloat, defaultValue=0, description=""}
  User.RibbonIDOffset {type=NiagaraInt32}
  User.SkeletalMesh {
   type=NiagaraDataInterfaceSkeletalMesh
   defaultValue {
    struct=NiagaraExt_VariableValue_DataInterface
    value {dataInterface=&self:NiagaraDataInterfaceSkeletalMesh_0}
   }
  }
  User.SpawnCount {type=NiagaraInt32}
 ]
}
